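# Unlit, two-sided masked material for Niagara mesh particles, apparently the lightning in the green EmeraldGlass emote effect.
# Tiled, time-panned UVs, reduced to their U component, drive vector noise that particle alpha thresholds into a hard, flickering cutout.
# Particle color, scaled by time-of-day HDR multipliers, feeds an emissive adjustment function.
# Further branches scale the noise by VAMul and build a dynamic-parameter V wipe, though their material outputs are not shown.
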Material /Game/Effects/Fort_Effects/Emotes/EmeraldGlass_Green/M_EmeraldGlass_D_Lightning_mesh
graph {
 properties {
  bCastRayTracedShadows=false
  bUseLightmapDirectionality=false
  bUsedWithNiagaraMeshParticles=true
  blendMode=BLEND_Masked
  pixelDepthOffsetMode=PDOM_Legacy
  shadingModel=MSM_Unlit
  twoSided=true
  usedShadingModels=Unlit
 }
 nodes [
  n1 = TextureCoordinate() {uTiling=10, vTiling=10}
  n2 = AppendVector(A=ScalarParameter() {defaultValue=10, parameterName=U scale}, B=ScalarParameter() {defaultValue=10, parameterName=V Scale})
  n3 = Time()
  n4 = Constant2Vector() {g=20, r=30}
  n5 = MaterialFunctionCall(Float2=Add(A=Multiply(A=@n1, B=@n2), B=Multiply(A=@n3, B=@n4))) {materialFunction=&BreakOutFloat2Components}
  n6 = VectorNoise(World Position=MaterialFunctionCall(X=@n5.R, Y=@n5.R, Z=Add(A=@n5.R, B=@n5.R)) {materialFunction=&MakeFloat3}.Result)
  n7 = ComponentMask(None=@n6) {g=true}
  n8 = ParticleColor()
  Ceil(None=Saturate(None=Subtract(A=@n7, B=@n8.A)))
  n9 = ScalarParameter() {defaultValue=0.7, parameterName=HDRMult_3}
  n10 = ScalarParameter() {defaultValue=1, parameterName=HDRMult_2}
  n11 = ScalarParameter() {defaultValue=0.7, parameterName=HDRMult_1}
  n12 = ScalarParameter() {defaultValue=0.5, parameterName=HDRMult}
  MaterialFunctionCall(Emissive=Multiply(
    A=MaterialFunctionCall(Color_Morning=@n9, Color_Day=@n10, Color_Evening=@n11, Color_Night=@n12) {materialFunction=&Color_Time_of_Day}.Result
    B=@n8.RGB
   )) {materialFunction=&MF_FX_EmissiveAdjustments}
  Multiply(A=@n7, B=@n8.A)
  n13 = OneMinus(None=ComponentMask(None=TextureCoordinate()) {g=true})
  n14 = DynamicParameter() {defaultValue=<rgba 0 1 1 1>, paramNames [UV V Anim, Param2, Param3, Param4]}
  Multiply(A=@n8.A, B=Sign(None=Add(A=@n13, B=@n14.UV V Anim)))
  Multiply(A=@n6, B=ScalarParameter() {defaultValue=3, parameterName=VAMul}) {constB=10}
 ]
}
